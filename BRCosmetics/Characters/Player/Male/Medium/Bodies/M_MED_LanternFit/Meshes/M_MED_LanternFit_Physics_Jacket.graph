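# Physics asset for the LanternFit medium male body's jacket, setting up dynamic jacket bones around pelvis, spine and thigh capsules.
# Nine capsule chains run from tiny MSTR root bodies, joined to spine_01 with free-angle constraints, through segments with tight twist and asymmetric swing limits.
# Free-angle cross constraints link neighboring second segments across back and sides, leaving the two three-segment front chains outside that band.
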PhysicsAsset /BRCosmetics/Characters/Player/Male/Medium/Bodies/M_MED_LanternFit/Meshes/M_MED_LanternFit_Physics_Jacket
bodies {
 dyn_jacket_bk_01 [{shapeType=Capsule, length=12, radius=2, shapeName=dyn_jacket_bk_01_capsule, center=<12 0 0>, rotation=<rot 0 0 90>}]
 dyn_jacket_bk_02 [{shapeType=Capsule, length=10, radius=2.2, shapeName=dyn_jacket_bk_02_capsule, center=<6 0 0>, rotation=<rot 0 0 90>}]
 dyn_jacket_bk_MSTR [{shapeType=Capsule, length=0.505, radius=0.505, shapeName=dyn_jacket_bk_MSTR_capsule, rotation=<rot 0 -90 0>}]
 dyn_jacket_l_01 [{shapeType=Capsule, length=12, radius=1.8, shapeName=dyn_jacket_l_01_capsule, center=<12.5599 0 0>, rotation=<rot 0 0.0001 80>}]
 dyn_jacket_l_02 [
  {shapeType=Capsule, length=14, radius=1.9, shapeName=dyn_jacket_l_02_capsule, center=<9.1457 0 0>, rotation=<rot 0 0 -95>}
  {shapeType=Capsule, length=14, radius=1.7, shapeName=dyn_jacket_l_02_capsule, center=<3 0 0>, rotation=<rot 0 0 -95>}
 ]
 dyn_jacket_l_MSTR [{shapeType=Capsule, length=0.505, radius=0.505, shapeName=dyn_jacket_l_MSTR_capsule, rotation=<rot 0 -90 0>}]
 dyn_jacket_l_bk_01 [{shapeType=Capsule, length=12, radius=2, shapeName=dyn_jacket_l_bk_01_capsule, center=<12 0 0>, rotation=<rot 0 0 90>}]
 dyn_jacket_l_bk_02 [{shapeType=Capsule, length=13, radius=2, shapeName=dyn_jacket_l_bk_02_capsule, center=<6.5 -0.14 0.3775>, rotation=<rot 0 0 100>}]
 dyn_jacket_l_bk_MSTR [{shapeType=Capsule, length=0.505, radius=0.505, shapeName=dyn_jacket_l_bk_MSTR_capsule, rotation=<rot 0 -90 0>}]
 dyn_jacket_l_fr_01 [{shapeType=Capsule, length=9, radius=2.2, shapeName=dyn_jacket_l_fr_01_capsule, center=<12.5599 -1.2548 0>, rotation=<rot 0 0 70>}]
 dyn_jacket_l_fr_02 [{shapeType=Capsule, length=12, radius=2, shapeName=dyn_jacket_l_fr_02_capsule, center=<4 -2.5487 -0.6829>, rotation=<rot 0 0 75>}]
 dyn_jacket_l_fr_03 [{shapeType=Capsule, length=4, radius=2.7, shapeName=dyn_jacket_l_fr_03_capsule, center=<5 0 0>, rotation=<rot 0 0 80>}]
 dyn_jacket_l_fr_MSTR [{shapeType=Capsule, length=0.505, radius=0.505, shapeName=dyn_jacket_l_fr_MSTR_capsule, rotation=<rot 0 -90 0>}]
 dyn_jacket_l_fr_l_01 [{shapeType=Capsule, length=12, radius=1.8, shapeName=dyn_jacket_l_fr_01_capsule, center=<12.5599 0 0>, rotation=<rot 0 0 90>}]
 dyn_jacket_l_fr_l_02 [
  {shapeType=Capsule, length=12, radius=2, shapeName=dyn_jacket_l_fr_02_capsule, center=<9.1457 0 0>, rotation=<rot 0 0 90>}
  {shapeType=Capsule, length=12, radius=1.8, shapeName=dyn_jacket_l_fr_l_02_capsule, center=<3 0 0>, rotation=<rot 0 0 90>}
 ]
 dyn_jacket_l_fr_l_MSTR [{shapeType=Capsule, length=0.505, radius=0.505, shapeName=dyn_jacket_l_fr_l_MSTR_capsule, rotation=<rot 0 -90 0>}]
 dyn_jacket_r_01 [{shapeType=Capsule, length=12, radius=1.8, shapeName=dyn_jacket_r_01_capsule, center=<12.5599 0 0>, rotation=<rot 0 0 -80>}]
 dyn_jacket_r_02 [
  {shapeType=Capsule, length=14, radius=1.9, shapeName=dyn_jacket_r_02_capsule, center=<9.1457 0 0>, rotation=<rot 0 0 95>}
  {shapeType=Capsule, length=14, radius=1.7, shapeName=dyn_jacket_r_02_capsule, center=<3 0 0>, rotation=<rot 0 0 95>}
 ]
 dyn_jacket_r_MSTR [{shapeType=Capsule, length=0.505, radius=0.505, shapeName=dyn_jacket_r_MSTR_capsule, rotation=<rot 0 -90 0>}]
 dyn_jacket_r_bk_01 [{shapeType=Capsule, length=12, radius=2, shapeName=dyn_jacket_r_bk_01_capsule, center=<12 0 0>, rotation=<rot 0 0 90>}]
 dyn_jacket_r_bk_02 [{shapeType=Capsule, length=13, radius=2, shapeName=dyn_jacket_r_bk_02_capsule, center=<6.5 0.14 0.3775>, rotation=<rot 0 0 -100>}]
 dyn_jacket_r_bk_MSTR [{shapeType=Capsule, length=0.505, radius=0.505, shapeName=dyn_jacket_r_bk_MSTR_capsule, rotation=<rot 0 -90 0>}]
 dyn_jacket_r_fr_01 [{shapeType=Capsule, length=9, radius=2.2, shapeName=dyn_jacket_r_fr_01_capsule, center=<12.5599 1.2548 0>, rotation=<rot 0 0 110>}]
 dyn_jacket_r_fr_02 [{shapeType=Capsule, length=12, radius=2, shapeName=dyn_jacket_r_fr_02_capsule, center=<4 2.5487 -0.6829>, rotation=<rot 0 0 -75>}]
 dyn_jacket_r_fr_03 [{shapeType=Capsule, length=4, radius=2.7, shapeName=dyn_jacket_r_fr_03_capsule, center=<5 0 0>, rotation=<rot 0 0 -80>}]
 dyn_jacket_r_fr_MSTR [{shapeType=Capsule, length=0.505, radius=0.505, shapeName=dyn_jacket_r_fr_MSTR_capsule, rotation=<rot 0 -90 0>}]
 dyn_jacket_r_fr_r_01 [{shapeType=Capsule, length=12, radius=1.8, shapeName=dyn_jacket_r_fr_01_capsule, center=<12.5599 0 0>, rotation=<rot 0 0 90>}]
 dyn_jacket_r_fr_r_02 [
  {shapeType=Capsule, length=12, radius=2, shapeName=dyn_jacket_r_fr_02_capsule, center=<9.1457 0 0>, rotation=<rot 0 0 90>}
  {shapeType=Capsule, length=12, radius=1.8, shapeName=dyn_jacket_r_fr_r_02_capsule, center=<3 0 0>, rotation=<rot 0 0 90>}
 ]
 dyn_jacket_r_fr_r_MSTR [{shapeType=Capsule, length=0.505, radius=0.505, shapeName=dyn_jacket_r_fr_r_MSTR_capsule, rotation=<rot 0 -90 0>}]
 pelvis [
  {shapeType=Capsule, length=13.0428, radius=9.7, shapeName=pelvis_capsule, center=<0 -1.3 5.8>, rotation=<rot 90 0 0>}
  {shapeType=Capsule, length=13.0428, radius=9.7, shapeName=pelvis_capsule, center=<0 -1.3 -5.8>, rotation=<rot 90 0 0>}
 ]
 spine_01 [
  {shapeType=Capsule, length=11.5208, radius=8.9, shapeName=spine_01_capsule, center=<0 0 6>, rotation=<rot 90 0 0>}
  {shapeType=Capsule, length=11.5208, radius=8.9, shapeName=spine_01_capsule, center=<0 0 -6>, rotation=<rot 90 0 0>}
 ]
 thigh_l [
  {shapeType=Capsule, length=28, radius=8.2, shapeName=thigh_l_capsule, center=<-13 0 0>, rotation=<rot 90 0 0>}
  {shapeType=Capsule, length=27, radius=7, shapeName=thigh_l_capsule, center=<-13 4 8>, rotation=<rot 90 0 0>}
 ]
 thigh_r [{shapeType=Capsule, length=28, radius=8.2, shapeName=thigh_r_capsule, center=<13 0 0>, rotation=<rot 90 180 0>}]
}
constraints [
 {
  bone1Name=dyn_jacket_bk_MSTR
  bone2Name=spine_01
  swing1LimitDegrees=45
  swing1Motion=Free
  swing2LimitDegrees=45
  swing2Motion=Free
  twistLimitDegrees=45
  twistMotion=Free
 }
 {bone1Name=dyn_jacket_bk_01, bone2Name=dyn_jacket_bk_MSTR, swing1LimitDegrees=6, swing2LimitDegrees=15, twistLimitDegrees=5}
 {
  bone1Name=dyn_jacket_l_MSTR
  bone2Name=spine_01
  swing1LimitDegrees=45
  swing1Motion=Free
  swing2LimitDegrees=45
  swing2Motion=Free
  twistLimitDegrees=45
  twistMotion=Free
 }
 {bone1Name=dyn_jacket_l_01, bone2Name=dyn_jacket_l_MSTR, swing1LimitDegrees=8, swing2LimitDegrees=30, twistLimitDegrees=5}
 {bone1Name=dyn_jacket_l_02, bone2Name=dyn_jacket_l_01, swing1LimitDegrees=4, swing2LimitDegrees=30, twistLimitDegrees=4}
 {
  bone1Name=dyn_jacket_l_bk_MSTR
  bone2Name=spine_01
  swing1LimitDegrees=45
  swing1Motion=Free
  swing2LimitDegrees=45
  swing2Motion=Free
  twistLimitDegrees=45
  twistMotion=Free
 }
 {bone1Name=dyn_jacket_l_bk_01, bone2Name=dyn_jacket_l_bk_MSTR, swing1LimitDegrees=6, swing2LimitDegrees=12, twistLimitDegrees=5}
 {bone1Name=dyn_jacket_l_bk_02, bone2Name=dyn_jacket_l_bk_01, swing1LimitDegrees=4, swing2LimitDegrees=17, twistLimitDegrees=4}
 {
  bone1Name=dyn_jacket_l_fr_MSTR
  bone2Name=spine_01
  swing1LimitDegrees=45
  swing1Motion=Free
  swing2LimitDegrees=45
  swing2Motion=Free
  twistLimitDegrees=45
  twistMotion=Free
 }
 {bone1Name=dyn_jacket_l_fr_01, bone2Name=dyn_jacket_l_fr_MSTR, swing1LimitDegrees=10, swing2LimitDegrees=45, twistLimitDegrees=8}
 {bone1Name=dyn_jacket_l_fr_02, bone2Name=dyn_jacket_l_fr_01, swing1LimitDegrees=10, swing2LimitDegrees=50, twistLimitDegrees=8}
 {
  bone1Name=dyn_jacket_r_MSTR
  bone2Name=spine_01
  swing1LimitDegrees=45
  swing1Motion=Free
  swing2LimitDegrees=45
  swing2Motion=Free
  twistLimitDegrees=45
  twistMotion=Free
 }
 {bone1Name=dyn_jacket_r_01, bone2Name=dyn_jacket_r_MSTR, swing1LimitDegrees=8, swing2LimitDegrees=30, twistLimitDegrees=5}
 {bone1Name=dyn_jacket_r_02, bone2Name=dyn_jacket_r_01, swing1LimitDegrees=4, swing2LimitDegrees=30, twistLimitDegrees=4}
 {
  bone1Name=dyn_jacket_r_bk_MSTR
  bone2Name=spine_01
  swing1LimitDegrees=45
  swing1Motion=Free
  swing2LimitDegrees=45
  swing2Motion=Free
  twistLimitDegrees=45
  twistMotion=Free
 }
 {bone1Name=dyn_jacket_r_bk_01, bone2Name=dyn_jacket_r_bk_MSTR, swing1LimitDegrees=6, swing2LimitDegrees=12, twistLimitDegrees=5}
 {bone1Name=dyn_jacket_r_bk_02, bone2Name=dyn_jacket_r_bk_01, swing1LimitDegrees=4, swing2LimitDegrees=17, twistLimitDegrees=4}
 {
  bone1Name=dyn_jacket_r_fr_MSTR
  bone2Name=spine_01
  swing1LimitDegrees=45
  swing1Motion=Free
  swing2LimitDegrees=45
  swing2Motion=Free
  twistLimitDegrees=45
  twistMotion=Free
 }
 {bone1Name=dyn_jacket_r_fr_01, bone2Name=dyn_jacket_r_fr_MSTR, swing1LimitDegrees=10, swing2LimitDegrees=45, twistLimitDegrees=8}
 {bone1Name=dyn_jacket_r_fr_02, bone2Name=dyn_jacket_r_fr_01, swing1LimitDegrees=10, swing2LimitDegrees=50, twistLimitDegrees=8}
 {
  bone1Name=dyn_jacket_l_fr_l_MSTR
  bone2Name=spine_01
  swing1LimitDegrees=45
  swing1Motion=Free
  swing2LimitDegrees=45
  swing2Motion=Free
  twistLimitDegrees=45
  twistMotion=Free
 }
 {bone1Name=dyn_jacket_l_fr_l_01, bone2Name=dyn_jacket_l_fr_l_MSTR, swing1LimitDegrees=10, swing2LimitDegrees=30, twistLimitDegrees=5}
 {bone1Name=dyn_jacket_l_fr_l_02, bone2Name=dyn_jacket_l_fr_l_01, swing1LimitDegrees=8, swing2LimitDegrees=30, twistLimitDegrees=3}
 {
  bone1Name=dyn_jacket_r_fr_r_MSTR
  bone2Name=spine_01
  swing1LimitDegrees=45
  swing1Motion=Free
  swing2LimitDegrees=45
  swing2Motion=Free
  twistLimitDegrees=45
  twistMotion=Free
 }
 {bone1Name=dyn_jacket_r_fr_r_01, bone2Name=dyn_jacket_r_fr_r_MSTR, swing1LimitDegrees=10, swing2LimitDegrees=30, twistLimitDegrees=5}
 {bone1Name=dyn_jacket_r_fr_r_02, bone2Name=dyn_jacket_r_fr_r_01, swing1LimitDegrees=8, swing2LimitDegrees=30, twistLimitDegrees=3}
 {
  bone1Name=spine_01
  bone2Name=pelvis
  swing1LimitDegrees=45
  swing1Motion=Free
  swing2LimitDegrees=45
  swing2Motion=Free
  twistLimitDegrees=45
  twistMotion=Free
 }
 {
  bone1Name=thigh_l
  bone2Name=pelvis
  swing1LimitDegrees=45
  swing1Motion=Free
  swing2LimitDegrees=45
  swing2Motion=Free
  twistLimitDegrees=45
  twistMotion=Free
 }
 {
  bone1Name=thigh_r
  bone2Name=pelvis
  swing1LimitDegrees=45
  swing1Motion=Free
  swing2LimitDegrees=45
  swing2Motion=Free
  twistLimitDegrees=45
  twistMotion=Free
 }
 {bone1Name=dyn_jacket_bk_02, bone2Name=dyn_jacket_bk_01, swing1LimitDegrees=6, swing2LimitDegrees=20, twistLimitDegrees=5}
 {bone1Name=dyn_jacket_r_fr_03, bone2Name=dyn_jacket_r_fr_02, swing1LimitDegrees=12, swing2LimitDegrees=55, twistLimitDegrees=10}
 {bone1Name=dyn_jacket_l_fr_03, bone2Name=dyn_jacket_l_fr_02, swing1LimitDegrees=12, swing2LimitDegrees=55, twistLimitDegrees=10}
 {
  bone1Name=dyn_jacket_l_bk_02
  bone2Name=dyn_jacket_bk_02
  swing1LimitDegrees=45
  swing1Motion=Free
  swing2LimitDegrees=45
  swing2Motion=Free
  twistLimitDegrees=45
  twistMotion=Free
 }
 {
  bone1Name=dyn_jacket_r_bk_02
  bone2Name=dyn_jacket_bk_02
  swing1LimitDegrees=45
  swing1Motion=Free
  swing2LimitDegrees=45
  swing2Motion=Free
  twistLimitDegrees=45
  twistMotion=Free
 }
 {
  bone1Name=dyn_jacket_l_02
  bone2Name=dyn_jacket_l_bk_02
  swing1LimitDegrees=45
  swing1Motion=Free
  swing2LimitDegrees=45
  swing2Motion=Free
  twistLimitDegrees=45
  twistMotion=Free
 }
 {
  bone1Name=dyn_jacket_r_02
  bone2Name=dyn_jacket_r_bk_02
  swing1LimitDegrees=45
  swing1Motion=Free
  swing2LimitDegrees=45
  swing2Motion=Free
  twistLimitDegrees=45
  twistMotion=Free
 }
 {
  bone1Name=dyn_jacket_r_fr_r_02
  bone2Name=dyn_jacket_r_02
  swing1LimitDegrees=45
  swing1Motion=Free
  swing2LimitDegrees=45
  swing2Motion=Free
  twistLimitDegrees=45
  twistMotion=Free
 }
 {
  bone1Name=dyn_jacket_l_fr_l_02
  bone2Name=dyn_jacket_l_02
  swing1LimitDegrees=45
  swing1Motion=Free
  swing2LimitDegrees=45
  swing2Motion=Free
  twistLimitDegrees=45
  twistMotion=Free
 }
]
