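# Post-process style AnimBlueprint for the M_MED_LanternFit male body cosmetic, adding secondary motion to its jacket, collar, zipper and shoe strings.
# Parent mesh pose feeds a deform Control Rig, then one of two jacket RigidBody setups, blended by backpack/cape detection.
# AnimDynamics and additive jacket bone offsets follow; bottom shoe-string nodes are unconnected.
# Event graph updates gravity, front-end versus gameplay sim-space settings, and crouch/DBNO values.
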AnimBlueprint /BRCosmetics/Characters/Player/Male/Medium/Bodies/M_MED_LanternFit/Meshes/M_MED_LanternFit_AnimBP
target_skeleton=&/M_MED_LanternFit_Skeleton
graphs [
 AnimGraph {nodes [
   Anim Dynamics - Bone: dyn_shoe_strings_bot_r_l!(ComponentPose, Alpha=1, GravityScale:float=1, GravityOverride:Vector="0,0,-490", ExternalForce="0,0,0") -> (Pose) {
    angularDampingOverride=0.9999
    angularSpringConstant=100
    bAngularSpring=true
    bGravityOverrideInSimSpace=true
    bOverrideAngularDamping=true
    bUseGravityOverride=true
    bUsePlanarLimit=false
    simulationSpace=RootRelative
    alphaBoolBlend {blendOption=Linear}
    boundBone {boneName=dyn_shoe_strings_bot_r_l}
    componentAppliedLinearAccClamp=<5000 5000 5000>
    componentLinearAccScale=<0.6 0.6 0.6>
    componentLinearVelScale=<0.6 0.6 0.6>
    gravityOverride {z=-980}
    physicsBodyDefinitions [{
      sphereCollisionRadius=0
      boundBone {boneName=dyn_shoe_strings_bot_r_l}
      boxExtents=<14.28 6.86 7.14>
      constraintSetup {angularLimitsMax=<7 5 0>, angularLimitsMin=<-10 -12 -70>, angularTarget=<1 0 0>}
      localJointOffset=<3.8752 0 0>
     }]
    relativeSpaceBone {boneName=pelvis}
   }
   Anim Dynamics - Bone: dyn_shoe_strings_bot_l_l!(ComponentPose, Alpha=1, GravityScale:float=1, GravityOverride:Vector="0,0,-490", ExternalForce="0,0,0") -> (Pose) {
    angularDampingOverride=0.9999
    angularSpringConstant=100
    bAngularSpring=true
    bGravityOverrideInSimSpace=true
    bOverrideAngularDamping=true
    bUseGravityOverride=true
    bUsePlanarLimit=false
    simulationSpace=RootRelative
    alphaBoolBlend {blendOption=Linear}
    boundBone {boneName=dyn_shoe_strings_bot_l_l}
    componentAppliedLinearAccClamp=<5000 5000 5000>
    componentLinearAccScale=<0.6 0.6 0.6>
    componentLinearVelScale=<0.6 0.6 0.6>
    gravityOverride {z=-980}
    physicsBodyDefinitions [{
      sphereCollisionRadius=0
      boundBone {boneName=dyn_shoe_strings_bot_l_l}
      boxExtents=<13.58 7.14 6.86>
      constraintSetup {angularLimitsMax=<7 5 0>, angularLimitsMin=<-10 -12 -70>, angularTarget=<1 0 0>}
      localJointOffset=<3.528 0 0>
     }]
    relativeSpaceBone {boneName=pelvis}
   }
   Anim Dynamics - Bone: dyn_shoe_strings_bot_l_r!(ComponentPose, Alpha=1, GravityScale:float=1, GravityOverride:Vector="0,0,-980", ExternalForce="0,0,0") -> (Pose) {
    angularDampingOverride=0.9999
    angularSpringConstant=100
    bAngularSpring=true
    bGravityOverrideInSimSpace=true
    bOverrideAngularDamping=true
    bUseGravityOverride=true
    bUsePlanarLimit=false
    simulationSpace=RootRelative
    alphaBoolBlend {blendOption=Linear}
    boundBone {boneName=dyn_shoe_strings_bot_l_r}
    componentAppliedLinearAccClamp=<5000 5000 5000>
    componentLinearAccScale=<0.6 0.6 0.6>
    componentLinearVelScale=<0.6 0.6 0.6>
    gravityOverride {z=-980}
    physicsBodyDefinitions [{
      sphereCollisionRadius=0
      boundBone {boneName=dyn_shoe_strings_bot_l_r}
      boxExtents=<13.58 7.14 6.86>
      constraintSetup {angularLimitsMax=<7 5 0>, angularLimitsMin=<-10 -12 -70>, angularTarget=<0 0 0>}
      localJointOffset=<3.528 0 0>
     }]
    relativeSpaceBone {boneName=foot_l}
   }
   Anim Dynamics - Bone: dyn_shoe_strings_bot_r_r!(ComponentPose, Alpha=1, GravityScale:float=1, GravityOverride:Vector="0,0,-980", ExternalForce="0,0,0") -> (Pose) {
    angularDampingOverride=0.9999
    angularSpringConstant=100
    bAngularSpring=true
    bGravityOverrideInSimSpace=true
    bOverrideAngularDamping=true
    bUseGravityOverride=true
    bUsePlanarLimit=false
    simulationSpace=RootRelative
    alphaBoolBlend {blendOption=Linear}
    boundBone {boneName=dyn_shoe_strings_bot_r_r}
    componentAppliedLinearAccClamp=<5000 5000 5000>
    componentLinearAccScale=<0.6 0.6 0.6>
    componentLinearVelScale=<0.6 0.6 0.6>
    gravityOverride {z=-980}
    physicsBodyDefinitions [{
      sphereCollisionRadius=0
      boundBone {boneName=dyn_shoe_strings_bot_r_r}
      boxExtents=<14.28 6.86 7.14>
      constraintSetup {angularLimitsMax=<7 5 0>, angularLimitsMin=<-10 -12 -70>, angularTarget=<0 0 0>}
      localJointOffset=<3.724 0 0>
     }]
    relativeSpaceBone {boneName=foot_l}
   }
   n1 = Anim Dynamics - Bone: dyn_shoe_strings_top_l_l!(
    ComponentPose=Local To Component(LocalPose=Use cached pose 'RigidBody'())
    Alpha=1
    GravityScale:float=1
    GravityOverride:Vector="0,0,-490"
    SimSpaceSettings:Anim Phys Sim Space Settings Structure=(SimSpaceAngularAlpha=0,MaxAngularVelocity=10000,MaxAngularAcceleration=10000,ExternalAngularVelocity=(X=0,Y=0,Z=0))
    ExternalForce="0,0,0"
   ) -> (Pose) {
    angularDampingOverride=0.9999
    angularSpringConstant=500
    bAngularSpring=true
    bGravityOverrideInSimSpace=true
    bOverrideAngularDamping=true
    bUseGravityOverride=true
    bUsePlanarLimit=false
    simulationSpace=RootRelative
    alphaBoolBlend {blendOption=Linear}
    boundBone {boneName=dyn_shoe_strings_top_l_l}
    componentAppliedLinearAccClamp=<5000 5000 5000>
    componentLinearAccScale=<0.6 0.6 0.6>
    componentLinearVelScale=<0.6 0.6 0.6>
    gravityOverride {z=-980}
    physicsBodyDefinitions [{
      sphereCollisionRadius=0
      boundBone {boneName=dyn_shoe_strings_top_l_l}
      boxExtents=<10 5 5>
      constraintSetup {angularLimitsMax=<7 5 0>, angularLimitsMin=<-10 -12 -70>, angularTarget=<1 0 0>}
      localJointOffset=<3 0 0>
     }]
    relativeSpaceBone {boneName=pelvis}
   }
   n2 = Anim Dynamics - Bone: dyn_shoe_strings_top_r_l!(
    ComponentPose=@n1
    Alpha=1
    GravityScale:float=1
    GravityOverride:Vector="0,0,-490"
    SimSpaceSettings:Anim Phys Sim Space Settings Structure=(SimSpaceAngularAlpha=0,MaxAngularVelocity=10000,MaxAngularAcceleration=10000,ExternalAngularVelocity=(X=0,Y=0,Z=0))
    ExternalForce="0,0,0"
   ) -> (Pose) {
    angularDampingOverride=0.9999
    angularSpringConstant=500
    bAngularSpring=true
    bGravityOverrideInSimSpace=true
    bOverrideAngularDamping=true
    bUseGravityOverride=true
    bUsePlanarLimit=false
    simulationSpace=RootRelative
    alphaBoolBlend {blendOption=Linear}
    boundBone {boneName=dyn_shoe_strings_top_r_l}
    componentAppliedLinearAccClamp=<5000 5000 5000>
    componentLinearAccScale=<0.6 0.6 0.6>
    componentLinearVelScale=<0.6 0.6 0.6>
    gravityOverride {z=-980}
    physicsBodyDefinitions [{
      sphereCollisionRadius=0
      boundBone {boneName=dyn_shoe_strings_top_r_l}
      boxExtents=<10 5 5>
      constraintSetup {angularLimitsMax=<7 5 0>, angularLimitsMin=<-10 -12 -70>, angularTarget=<1 0 0>}
      localJointOffset=<3 0 0>
     }]
    relativeSpaceBone {boneName=pelvis}
   }
   n3 = Anim Dynamics - Bone: dyn_shoe_strings_top_l_r!(
    ComponentPose=@n2
    Alpha=1
    GravityScale:float=1
    GravityOverride:Vector="0,0,-980"
    SimSpaceSettings:Anim Phys Sim Space Settings Structure=(SimSpaceAngularAlpha=0,MaxAngularVelocity=10000,MaxAngularAcceleration=10000,ExternalAngularVelocity=(X=0,Y=0,Z=0))
    ExternalForce="0,0,0"
   ) -> (Pose) {
    angularDampingOverride=0.9999
    angularSpringConstant=500
    bAngularSpring=true
    bGravityOverrideInSimSpace=true
    bOverrideAngularDamping=true
    bUseGravityOverride=true
    bUsePlanarLimit=false
    simulationSpace=RootRelative
    alphaBoolBlend {blendOption=Linear}
    boundBone {boneName=dyn_shoe_strings_top_l_r}
    componentAppliedLinearAccClamp=<5000 5000 5000>
    componentLinearAccScale=<0.6 0.6 0.6>
    componentLinearVelScale=<0.6 0.6 0.6>
    gravityOverride {z=-980}
    physicsBodyDefinitions [{
      sphereCollisionRadius=0
      boundBone {boneName=dyn_shoe_strings_top_l_r}
      boxExtents=<10 5 5>
      constraintSetup {angularLimitsMax=<7 5 0>, angularLimitsMin=<-10 -12 -70>, angularTarget=<1 0 0>}
      localJointOffset=<3 0 0>
     }]
    relativeSpaceBone {boneName=foot_l}
   }
   n4 = Anim Dynamics - Bone: dyn_shoe_strings_top_r_r!(
    ComponentPose=@n3
    Alpha=1
    GravityScale:float=1
    GravityOverride:Vector="0,0,-980"
    SimSpaceSettings:Anim Phys Sim Space Settings Structure=(SimSpaceAngularAlpha=0,MaxAngularVelocity=10000,MaxAngularAcceleration=10000,ExternalAngularVelocity=(X=0,Y=0,Z=0))
    ExternalForce="0,0,0"
   ) -> (Pose) {
    angularBiasOverride=0.9999
    angularDampingOverride=0.9999
    angularSpringConstant=500
    bAngularSpring=true
    bGravityOverrideInSimSpace=true
    bOverrideAngularBias=true
    bOverrideAngularDamping=true
    bOverrideLinearDamping=true
    bUseGravityOverride=true
    bUsePlanarLimit=false
    linearDampingOverride=0.9999
    simulationSpace=RootRelative
    alphaBoolBlend {blendOption=Linear}
    boundBone {boneName=dyn_shoe_strings_top_r_r}
    componentAppliedLinearAccClamp=<5000 5000 5000>
    componentLinearAccScale=<0.6 0.6 0.6>
    componentLinearVelScale=<0.6 0.6 0.6>
    gravityOverride {z=-980}
    physicsBodyDefinitions [{
      sphereCollisionRadius=0
      boundBone {boneName=dyn_shoe_strings_top_r_r}
      boxExtents=<10 5 5>
      constraintSetup {angularLimitsMax=<7 5 0>, angularLimitsMin=<-10 -12 -70>, angularTarget=<1 0 0>}
      localJointOffset=<3 0 0>
     }]
    relativeSpaceBone {boneName=foot_l}
   }
   n5 = Anim Dynamics - Bone: dyn_zipper_01!(
    ComponentPose=@n4
    Alpha=1
    GravityOverride:Vector="0,0,-980"
    SimSpaceSettings:Anim Phys Sim Space Settings Structure=(SimSpaceAngularAlpha=0,MaxAngularVelocity=10000,MaxAngularAcceleration=10000,ExternalAngularVelocity=(X=0,Y=0,Z=0))
    ExternalForce="0,0,0"
   ) -> (Pose) {
    angularBiasOverride=0.9999
    angularDampingOverride=0.9999
    angularSpringConstant=500
    bOverrideAngularBias=true
    bOverrideAngularDamping=true
    bOverrideLinearDamping=true
    bUseGravityOverride=true
    bUsePlanarLimit=false
    gravityScale=-1
    linearDampingOverride=0.9999
    simulationSpace=RootRelative
    alphaBoolBlend {blendOption=Linear}
    boundBone {boneName=dyn_zipper_01}
    componentAppliedLinearAccClamp=<5000 5000 5000>
    componentLinearAccScale=<0.2 0.2 0.2>
    componentLinearVelScale=<0.2 0.2 0.2>
    gravityOverride {z=-980}
    physicsBodyDefinitions [{
      collisionType=CustomSphere
      sphereCollisionRadius=1
      boundBone {boneName=dyn_zipper_01}
      boxExtents=<10 5 5>
      constraintSetup {angularLimitsMax=<2 35 0>, angularLimitsMin=<-2 -35 -170>, angularTarget=<0 0 0>}
      localJointOffset=<3 0 0>
     }]
   }
   n6 = Anim Dynamics - Bone: dyn_collar_bot_l_01!(
    ComponentPose=@n5
    Alpha=1
    GravityOverride:Vector="0,0,-980"
    SimSpaceSettings:Anim Phys Sim Space Settings Structure=(SimSpaceAngularAlpha=0,MaxAngularVelocity=10000,MaxAngularAcceleration=10000,ExternalAngularVelocity=(X=0,Y=0,Z=0))
    ExternalForce="0,0,0"
   ) -> (Pose) {
    angularDampingOverride=0.9999
    angularSpringConstant=3000
    bAngularSpring=true
    bOverrideAngularDamping=true
    bUseGravityOverride=true
    bUsePlanarLimit=false
    gravityScale=-1
    simulationSpace=RootRelative
    alphaBoolBlend {blendOption=Linear}
    boundBone {boneName=dyn_collar_bot_l_01}
    componentAppliedLinearAccClamp=<5000 5000 5000>
    componentLinearAccScale=<0.2 0.2 0.2>
    componentLinearVelScale=<0.2 0.2 0.2>
    gravityOverride {z=-980}
    physicsBodyDefinitions [{
      collisionType=CustomSphere
      sphereCollisionRadius=1
      boundBone {boneName=dyn_collar_bot_l_01}
      boxExtents=<7.5 7.5 15>
      constraintSetup {angularLimitsMax=<0 0 35>, angularLimitsMin=<0 0 -5>, angularTarget=<1 0 0>}
      localJointOffset=<4 0 0>
     }]
   }
   n7 = Anim Dynamics - Bone: dyn_collar_bot_r_01!(
    ComponentPose=@n6
    Alpha=1
    GravityOverride:Vector="0,0,-980"
    SimSpaceSettings:Anim Phys Sim Space Settings Structure=(SimSpaceAngularAlpha=0,MaxAngularVelocity=10000,MaxAngularAcceleration=10000,ExternalAngularVelocity=(X=0,Y=0,Z=0))
    ExternalForce="0,0,0"
   ) -> (Pose) {
    angularDampingOverride=0.9999
    angularSpringConstant=3500
    bAngularSpring=true
    bOverrideAngularDamping=true
    bUseGravityOverride=true
    bUsePlanarLimit=false
    gravityScale=-1
    simulationSpace=RootRelative
    alphaBoolBlend {blendOption=Linear}
    boundBone {boneName=dyn_collar_bot_r_01}
    componentAppliedLinearAccClamp=<5000 5000 5000>
    componentLinearAccScale=<0.2 0.2 0.2>
    componentLinearVelScale=<0.2 0.2 0.2>
    gravityOverride {z=-980}
    physicsBodyDefinitions [{
      collisionType=CustomSphere
      sphereCollisionRadius=1
      boundBone {boneName=dyn_collar_bot_r_01}
      boxExtents=<10 10 20>
      constraintSetup {angularLimitsMax=<0 0 30>, angularLimitsMin=<0 0 -5>, angularTarget=<1 0 0>}
      localJointOffset=<5 0 0>
     }]
   }
   n8 = Anim Dynamics - Bone: dyn_collar_top_l_01!(
    ComponentPose=@n7
    Alpha=1
    GravityOverride:Vector="0,0,-980"
    SimSpaceSettings:Anim Phys Sim Space Settings Structure=(SimSpaceAngularAlpha=0,MaxAngularVelocity=10000,MaxAngularAcceleration=10000,ExternalAngularVelocity=(X=0,Y=0,Z=0))
    ExternalForce="0,0,0"
   ) -> (Pose) {
    angularBiasOverride=0.9999
    angularDampingOverride=0.9999
    angularSpringConstant=6500
    bAngularSpring=true
    bOverrideAngularBias=true
    bOverrideAngularDamping=true
    bOverrideLinearDamping=true
    bUseGravityOverride=true
    bUsePlanarLimit=false
    gravityScale=-1
    linearDampingOverride=0.9999
    simulationSpace=RootRelative
    alphaBoolBlend {blendOption=Linear}
    boundBone {boneName=dyn_collar_top_l_01}
    componentAppliedLinearAccClamp=<5000 5000 5000>
    componentLinearAccScale=<0.2 0.2 0.2>
    componentLinearVelScale=<0.2 0.2 0.2>
    gravityOverride {z=-980}
    physicsBodyDefinitions [{
      collisionType=CustomSphere
      sphereCollisionRadius=1
      boundBone {boneName=dyn_collar_top_l_01}
      boxExtents=<7.5 7.5 15>
      constraintSetup {angularLimitsMax=<0 0 10>, angularLimitsMin=<0 0 -15>, angularTarget=<1 0 0>}
      localJointOffset=<2 0 0>
     }]
   }
   n9 = Anim Dynamics - Bone: dyn_collar_top_r_01!(
    ComponentPose=@n8
    Alpha=1
    GravityOverride:Vector="0,0,-980"
    SimSpaceSettings:Anim Phys Sim Space Settings Structure=(SimSpaceAngularAlpha=0,MaxAngularVelocity=10000,MaxAngularAcceleration=10000,ExternalAngularVelocity=(X=0,Y=0,Z=0))
    ExternalForce="0,0,0"
   ) -> (Pose) {
    angularBiasOverride=0.9999
    angularDampingOverride=0.9999
    angularSpringConstant=6500
    bAngularSpring=true
    bOverrideAngularBias=true
    bOverrideAngularDamping=true
    bOverrideLinearDamping=true
    bUseGravityOverride=true
    bUsePlanarLimit=false
    gravityScale=-1
    linearDampingOverride=0.9999
    simulationSpace=RootRelative
    alphaBoolBlend {blendOption=Linear}
    boundBone {boneName=dyn_collar_top_r_01}
    componentAppliedLinearAccClamp=<5000 5000 5000>
    componentLinearAccScale=<0.2 0.2 0.2>
    componentLinearVelScale=<0.2 0.2 0.2>
    gravityOverride {z=-980}
    physicsBodyDefinitions [{
      collisionType=CustomSphere
      sphereCollisionRadius=1
      boundBone {boneName=dyn_collar_top_r_01}
      boxExtents=<7.5 7.5 15>
      constraintSetup {angularLimitsMax=<0 0 10>, angularLimitsMin=<0 0 -15>, angularTarget=<1 0 0>}
      localJointOffset=<2 0 0>
     }]
   }
   n10 = "Transform (Modify) Bone - Bone: dyn_jacket_bk_01"(ComponentPose=@n9, Alpha=1, Translation="0,-0.75,-0.5", Rotation="0,0,-1", Scale="1,1,1") {
    rotationMode=BMM_Additive
    translationMode=BMM_Additive
    alphaScaleBiasClamp {bInterpResult=true, interpSpeedDecreasing=5, interpSpeedIncreasing=5}
    boneToModify {boneName=dyn_jacket_bk_01}
    rotation {roll=-1}
    translation {y=-0.75, z=-0.5}
   }
   n11 = "Transform (Modify) Bone - Bone: dyn_jacket_bk_02"(ComponentPose=@n10, Alpha=1, Translation="0,-0.75,-0.5", Rotation="0,0,-9", Scale="1,1,1") {
    rotationMode=BMM_Additive
    translationMode=BMM_Additive
    alphaScaleBiasClamp {bInterpResult=true, interpSpeedDecreasing=5, interpSpeedIncreasing=5}
    boneToModify {boneName=dyn_jacket_bk_02}
    rotation {roll=-9}
    translation {y=-0.75, z=-0.5}
   }
   n12 = "Transform (Modify) Bone - Bone: dyn_jacket_l_bk_01"(ComponentPose=@n11, Alpha=1, Translation="0,-0.75,-0.5", Rotation="0,0,-1", Scale="1,1,1") {
    rotationMode=BMM_Additive
    translationMode=BMM_Additive
    alphaScaleBiasClamp {bInterpResult=true, interpSpeedDecreasing=5, interpSpeedIncreasing=5}
    boneToModify {boneName=dyn_jacket_l_bk_01}
    rotation {roll=-1}
    translation {y=-0.75, z=-0.5}
   }
   n13 = "Transform (Modify) Bone - Bone: dyn_jacket_l_bk_02"(ComponentPose=@n12, Alpha=1, Translation="0,-0.5,-0.75", Rotation="0,0,-9", Scale="1,1,1") {
    rotationMode=BMM_Additive
    translationMode=BMM_Additive
    alphaScaleBiasClamp {bInterpResult=true, interpSpeedDecreasing=5, interpSpeedIncreasing=5}
    boneToModify {boneName=dyn_jacket_l_bk_02}
    rotation {roll=-9}
    translation {y=-0.5, z=-0.75}
   }
   n14 = "Transform (Modify) Bone - Bone: dyn_jacket_r_bk_01"(ComponentPose=@n13, Alpha=1, Translation="0,-0.75,-0.5", Rotation="0,0,-1", Scale="1,1,1") {
    rotationMode=BMM_Additive
    translationMode=BMM_Additive
    alphaScaleBiasClamp {bInterpResult=true, interpSpeedDecreasing=5, interpSpeedIncreasing=5}
    boneToModify {boneName=dyn_jacket_r_bk_01}
    rotation {roll=-1}
    translation {y=-0.75, z=-0.5}
   }
   n15 = "Transform (Modify) Bone - Bone: dyn_jacket_r_bk_02"(ComponentPose=@n14, Alpha=1, Translation="0,-0.75,-0.75", Rotation="0,0,-9", Scale="1,1,1") {
    rotationMode=BMM_Additive
    translationMode=BMM_Additive
    alphaScaleBiasClamp {bInterpResult=true, interpSpeedDecreasing=5, interpSpeedIncreasing=5}
    boneToModify {boneName=dyn_jacket_r_bk_02}
    rotation {roll=-9}
    translation {y=-0.75, z=-0.75}
   }
   n16 = "Transform (Modify) Bone - Bone: dyn_jacket_r_fr_02"(ComponentPose=@n15, Alpha=1, Translation="0,0,0", Rotation="8,0,0", Scale="1,1,1") {
    rotationMode=BMM_Additive
    translationMode=BMM_Additive
    alphaScaleBiasClamp {bInterpResult=true, interpSpeedDecreasing=1, interpSpeedIncreasing=1}
    boneToModify {boneName=dyn_jacket_r_fr_02}
    rotation {pitch=8}
   }
   n17 = "Transform (Modify) Bone - Bone: dyn_jacket_r_fr_03"(ComponentPose=@n16, Alpha=1, Translation="0,0,0", Rotation="8,0,0", Scale="1,1,1") {
    rotationMode=BMM_Additive
    translationMode=BMM_Additive
    alphaScaleBiasClamp {bInterpResult=true, interpSpeedDecreasing=1, interpSpeedIncreasing=1}
    boneToModify {boneName=dyn_jacket_r_fr_03}
    rotation {pitch=8}
   }
   n18 = "Transform (Modify) Bone - Bone: dyn_jacket_l_fr_01"(ComponentPose=@n17, Alpha=1, Translation="-1,0,2", Rotation="0,0,0", Scale="1,1,1") {
    rotationMode=BMM_Additive
    translationMode=BMM_Additive
    alphaScaleBiasClamp {bInterpResult=true, interpSpeedDecreasing=1, interpSpeedIncreasing=1}
    boneToModify {boneName=dyn_jacket_l_fr_01}
    translation {x=-1, z=2}
   }
   n19 = "Transform (Modify) Bone - Bone: dyn_jacket_l_fr_02"(ComponentPose=@n18, Alpha=1, Translation="-1,0,0", Rotation="0,0,0", Scale="1,1,1") {
    rotationMode=BMM_Additive
    translationMode=BMM_Additive
    alphaScaleBiasClamp {bInterpResult=true, interpSpeedDecreasing=1, interpSpeedIncreasing=1}
    boneToModify {boneName=dyn_jacket_l_fr_02}
    translation {x=-1}
   }
   n20 = "Transform (Modify) Bone - Bone: dyn_jacket_l_fr_03"(ComponentPose=@n19, Alpha=1, Translation="-2,0,0", Rotation="45,15,-15", Scale="1,1,1") {
    rotationMode=BMM_Additive
    translationMode=BMM_Additive
    alphaScaleBiasClamp {bInterpResult=true, interpSpeedDecreasing=1, interpSpeedIncreasing=1}
    boneToModify {boneName=dyn_jacket_l_fr_03}
    rotation=<rot 45 15 -15>
    translation {x=-2}
   }
   n21 = "Transform (Modify) Bone - Bone: dyn_jacket_l_fr_l_02"(ComponentPose=@n20, Alpha=1, Translation="-2,0,0", Rotation="-10,40,0", Scale="1,1,1") {
    rotationMode=BMM_Additive
    translationMode=BMM_Additive
    alphaScaleBiasClamp {bInterpResult=true, interpSpeedDecreasing=1, interpSpeedIncreasing=1}
    boneToModify {boneName=dyn_jacket_l_fr_l_02}
    rotation {pitch=-10, yaw=40}
    translation {x=-2}
   }
   n22 = "Transform (Modify) Bone - Bone: dyn_jacket_l_02"(ComponentPose=@n21, Alpha=1, Translation="0,0,0", Rotation="-20,10,0", Scale="1,1,1") {
    rotationMode=BMM_Additive
    translationMode=BMM_Additive
    alphaScaleBiasClamp {bInterpResult=true, interpSpeedDecreasing=1, interpSpeedIncreasing=1}
    boneToModify {boneName=dyn_jacket_l_02}
    rotation {pitch=-20, yaw=10}
   }
   Output Pose(Result=Component To Local(ComponentPose=@n22)) {layerGroup=None}
   n23 = RigidBody!(ComponentPose=Local To Component(LocalPose=Use cached pose 'PreSim'()), Alpha, OverrideWorldGravity:Vector="0,0,0") -> (Pose) {
    bClampLinearTranslationLimitToRefPose=true
    bOverrideWorldGravity=true
    cachedBoundsScale=1
    overridePhysicsAsset=&/M_MED_LanternFit_Physics_Jacket_Back
    simulationSpace=BaseBoneSpace
    alphaBoolBlend {blendOption=Linear}
    baseBoneRef {boneName=spine_03}
    componentAppliedLinearAccClamp=<3000 3000 3000>
    simSpaceSettings {
     maxAngularAcceleration=20
     maxAngularVelocity=5
     maxLinearAcceleration=1200
     maxLinearVelocity=300
     velocityScaleZ=0.6
     worldAlpha=1
     externalLinearDragV=<1 1 1>
    }
   }
   n24 = RigidBody(ComponentPose=Local To Component(LocalPose=Use cached pose 'PreSim'()), OverrideWorldGravity="0,0,0") {
    bClampLinearTranslationLimitToRefPose=true
    bOverrideWorldGravity=true
    cachedBoundsScale=1
    overridePhysicsAsset=&/M_MED_LanternFit_Physics_Jacket
    simulationSpace=BaseBoneSpace
    alphaBoolBlend {blendOption=Linear}
    baseBoneRef {boneName=spine_03}
    componentAppliedLinearAccClamp=<3000 3000 3000>
    simSpaceSettings {
     maxAngularAcceleration=20
     maxAngularVelocity=5
     maxLinearAcceleration=1200
     maxLinearVelocity=300
     velocityScaleZ=0.6
     worldAlpha=1
     externalLinearDragV=<1 1 1>
    }
   }
   Save cached pose 'RigidBody'(Pose=Blend Poses by bool(
     BlendPose_0=Component To Local(ComponentPose=@n23)
     BlendPose_1=Component To Local(ComponentPose=@n24)
     BlendTime_0=0.7
     BlendTime_1=0.7
     bActiveValue=Get IsWearingBackpackOrCape() -> (IsWearingBackpackOrCape:bool)
    ) {blendType=Linear}) {cacheName=RigidBody}
   n25 = Control Rig(Source=Copy Pose From Mesh() {bUseAttachedParent=true}) {controlRigAssetReference {blueprintRigClass=&/M_MED_LanternFit_CtrlRig_Deform.M_MED_LanternFit_CtrlRig_Deform_C}}
   Save cached pose 'PreSim'(Pose=@n25) {cacheName=PreSim}
  ]}
 EventGraph {nodes [
   Event BlueprintUpdateAnimation()
   > n1 = Sequence!(execute) -> (then_0, then_1, then_2:Exec, then_3:Exec, then_4:Exec)
   > n2 = "Gravity Override State Function (Parameter Asset) (Nativized)"(
    CharacterPartAnimInstance=Self-Reference()
    GameTimeInSeconds=GetGameTimeInSeconds()
    Parameters=&/M_MED_LanternFit_Parameters
   )
   > Set GravityOverride(GravityOverride=@n2)
   > Set Character_Rotation(Character_Rotation:double=BreakRotator(InRot=Get PawnRotationRate() -> (PawnRotationRate:Rotator)).Yaw) -> (Output_Get:double)
   n3 = Branch(execute=@n1.then_3, Condition=Get bIsDBNO())
   > n4 = Set IsDBNO(IsDBNO:double=0.5) -> (Output_Get:double)
   Set IsCrouchORDBNO(
    execute=[@n4.then, @n12.then, @n13.then, @n5.then]
    IsCrouchORDBNO:double="Max (Float)"(A=Get IsCrouch() -> (IsCrouch:double), B=Get IsDBNO() -> (IsDBNO:double))
   ) -> (Output_Get:double)
   n5 = Set IsDBNO(execute=@n3.else, IsDBNO=0)
   n6 = Get bIsTargeting()
   n7 = Get bIsMoving2D()
   n8 = Get bIsPlayingMeleeAnim()
   n9 = Get Character_Rotation() -> (Character_Rotation:double)
   n10 = Branch(
    execute=@n1.then_4
    Condition=AND Boolean!(A=Get bIsCrouching(), B=NOT Boolean(A=@n7), C:bool=NOT Boolean(A=@n8), D:bool=NOT Boolean(A=@n6), E:bool="Equal (Float)"(A=@n9)) -> (ReturnValue)
   )
   > Set IsCrouch_AND_NotMove2D(IsCrouch_AND_NotMove2D:double=1) -> (Output_Get:double)
   Set IsCrouch_AND_NotMove2D(execute=@n10.else, IsCrouch_AND_NotMove2D=0)
   n11 = Branch(execute=@n1.then_2, Condition=Get bIsCrouching())
   > n12 = Set IsCrouch(IsCrouch:double=1) -> (Output_Get:double)
   n13 = Set IsCrouch(execute=@n11.else, IsCrouch=0)
   n14 = NOT Boolean(A=Get bIsPlayingEmote())
   n15 = NOT Boolean(A=Get bIsSkyDiving())
   n16 = Branch(execute=@n1.then_1, Condition=AND Boolean!(A=Get bIsInFrontEnd(), B=AND Boolean(A=@n15, B=@n14)) -> (ReturnValue))
   > Set SimSpaceSettings(SimSpaceSettings:Anim Phys Sim Space Settings Structure=Get SimSpaceSettingsFE() -> (SimSpaceSettingsFE:Anim Phys Sim Space Settings Structure)) -> (Output_Get:Anim Phys Sim Space Settings Structure)
   Set SimSpaceSettings(execute=@n16.else, SimSpaceSettings=Get SimSpaceSettingsPIE() -> (SimSpaceSettingsPIE:Anim Phys Sim Space Settings Structure))
   Event OnCharacterCustomizationLoaded()
   n17 = CheckForBackbling()
   > Set IsWearingBackpackOrCape(IsWearingBackpackOrCape:bool=OR Boolean!(A=@n17, B=@n17.bIsBackblingACape, C:bool=@n17.bIsBackblingLong) -> (ReturnValue)) -> (Output_Get:bool)
  ]}
]
variables [
 Character_Rotation {type=number}
 GravityOverride {type=Vector}
 IsCrouch {type=number}
 IsCrouch_AND_NotMove2D {type=number}
 IsCrouchORDBNO {type=number}
 IsDBNO {type=number}
 IsWearingBackpackOrCape {type=boolean}
 SimSpaceSettings {
  type=AnimPhysSimSpaceSettings
  default {maxAngularAcceleration=100000, maxAngularVelocity=100000, simSpaceAngularAlpha=1, externalAngularVelocity=<100000 100000 100000>}
 }
 SimSpaceSettingsFE {
  type=AnimPhysSimSpaceSettings
  default {maxAngularAcceleration=50, maxAngularVelocity=3, simSpaceAngularAlpha=1, externalAngularVelocity=<0 0 0>}
 }
 SimSpaceSettingsPIE {
  type=AnimPhysSimSpaceSettings
  default {maxAngularAcceleration=10, maxAngularVelocity=10, simSpaceAngularAlpha=0, externalAngularVelocity=<0 0 2>}
 }
]
