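# Post-process-style AnimBlueprint for the LanternFit face accessory, adding secondary motion to the part's dyn_fair bones.
# AnimGraph copies the attached parent's pose through seven serial spring-based AnimDynamics nodes (left 01–03, right 01–04); r_05, b_01 and l_04 nodes never reach the output.
# Each update sets GravityOverride using Gimmick's parameters and picks front-end or in-game SimSpaceSettings, excluding emotes and skydiving from front-end.
# The AnimGraph shows no links to either variable.
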AnimBlueprint /BRCosmetics/Characters/Player/Male/Medium/Bodies/M_MED_LanternFit/Meshes/Parts/M_MED_LanternFit_FaceAcc_AnimBP
target_skeleton=&/M_MED_LanternFit_FaceAcc_Skeleton
graphs [
 AnimGraph {nodes [
   n1 = Anim Dynamics - Bone: dyn_fair_r_05!(SimulationSpace:AnimPhysSimSpaceType Enum=Component, ComponentPose, Alpha=1, GravityOverride:Vector="0,0,0", ExternalForce="0,0,0") -> (Pose) {
    angularBiasOverride=0.9999
    angularDampingOverride=0.9999
    bLinearSpring=true
    bOverrideAngularBias=true
    bOverrideAngularDamping=true
    bOverrideLinearDamping=true
    bUseGravityOverride=true
    linearDampingOverride=0.9999
    linearSpringConstant=100
    simulationSpace=RootRelative
    alphaBoolBlend {blendOption=Linear}
    alphaScaleBiasClamp {scale=0.2}
    boundBone {boneName=dyn_fair_r_05}
    componentAppliedLinearAccClamp=<1250 1250 1250>
    componentLinearAccScale=<0.4 0.5 0.5>
    componentLinearVelScale=<0.4 0.5 0.5>
    physicsBodyDefinitions [{
      boundBone {boneName=dyn_fair_r_05}
      boxExtents=<8 8 12>
      constraintSetup {angularTargetAxis=AxisZ, angularLimitsMax=<2 10 0>, angularLimitsMin=<-2 -10 -8>, angularTarget=<1 1 1>}
      localJointOffset=<5 0 0>
     }]
   }
   Anim Dynamics - Bone: dyn_fair_b_01!(SimulationSpace:AnimPhysSimSpaceType Enum=Component, ComponentPose=@n1, Alpha=1, GravityOverride:Vector="0,0,0", ExternalForce="0,0,0") -> (Pose) {
    angularBiasOverride=0.9999
    angularDampingOverride=0.9999
    bLinearSpring=true
    bOverrideAngularBias=true
    bOverrideAngularDamping=true
    bOverrideLinearDamping=true
    bUseGravityOverride=true
    linearDampingOverride=0.9999
    linearSpringConstant=100
    simulationSpace=RootRelative
    alphaBoolBlend {blendOption=Linear}
    alphaScaleBiasClamp {scale=0.5}
    boundBone {boneName=dyn_fair_b_01}
    componentAppliedLinearAccClamp=<1250 1250 1250>
    componentLinearAccScale=<0.6 0.8 0.8>
    componentLinearVelScale=<0.6 0.8 0.8>
    physicsBodyDefinitions [{
      boundBone {boneName=dyn_fair_b_01}
      boxExtents=<8 8 12>
      constraintSetup {angularTargetAxis=AxisZ, angularLimitsMax=<2 4 0>, angularLimitsMin=<-2 -4 -1>, angularTarget=<1 1 1>}
      localJointOffset=<8 0 0>
     }]
   }
   Anim Dynamics - Bone: dyn_fair_l_04!(SimulationSpace:AnimPhysSimSpaceType Enum=Component, ComponentPose, Alpha=1, GravityOverride:Vector="0,0,0", ExternalForce="0,0,0") -> (Pose) {
    angularBiasOverride=0.9999
    angularDampingOverride=0.9999
    bLinearSpring=true
    bOverrideAngularBias=true
    bOverrideAngularDamping=true
    bOverrideLinearDamping=true
    bUseGravityOverride=true
    linearDampingOverride=0.9999
    linearSpringConstant=100
    simulationSpace=RootRelative
    alphaBoolBlend {blendOption=Linear}
    alphaScaleBiasClamp {scale=0.2}
    boundBone {boneName=dyn_fair_l_04}
    componentAppliedLinearAccClamp=<1250 1250 1250>
    componentLinearAccScale=<0.4 0.5 0.5>
    componentLinearVelScale=<0.4 0.5 0.5>
    physicsBodyDefinitions [{
      boundBone {boneName=dyn_fair_l_04}
      boxExtents=<8 8 12>
      constraintSetup {angularTargetAxis=AxisZ, angularLimitsMax=<2 5 0>, angularLimitsMin=<-2 -5 -6>, angularTarget=<1 1 1>}
      localJointOffset=<8 0 0>
     }]
   }
   n2 = Anim Dynamics - Bone: dyn_fair_l_01!(
    ComponentPose=Local To Component(LocalPose=Copy Pose From Mesh() {bUseAttachedParent=true})
    Alpha=1
    GravityOverride:Vector="0,0,0"
    SimSpaceSettings:Anim Phys Sim Space Settings Structure=(SimSpaceAngularAlpha=0,MaxAngularVelocity=10000,MaxAngularAcceleration=10000,ExternalAngularVelocity=(X=0,Y=0,Z=0))
    ExternalForce="0,0,0"
   ) -> (Pose) {
    angularBiasOverride=0.9999
    angularDampingOverride=0.9999
    bLinearSpring=true
    bOverrideAngularBias=true
    bOverrideAngularDamping=true
    bOverrideLinearDamping=true
    bUseGravityOverride=true
    linearDampingOverride=0.9999
    linearSpringConstant=100
    simulationSpace=RootRelative
    alphaBoolBlend {blendOption=Linear}
    alphaScaleBiasClamp {scale=0.72}
    boundBone {boneName=dyn_fair_l_01}
    componentAppliedLinearAccClamp=<1250 1250 1250>
    componentLinearAccScale=<0.4 0.5 0.5>
    componentLinearVelScale=<0.4 0.5 0.5>
    physicsBodyDefinitions [{
      boundBone {boneName=dyn_fair_l_01}
      boxExtents=<8 8 12>
      constraintSetup {angularTargetAxis=AxisZ, angularLimitsMax=<5 15 1>, angularLimitsMin=<-5 -15 -15>, angularTarget=<1 1 1>}
      localJointOffset=<3 0 0>
     }]
   }
   n3 = Anim Dynamics - Bone: dyn_fair_l_02!(
    ComponentPose=@n2
    Alpha=1
    GravityOverride:Vector="0,0,0"
    SimSpaceSettings:Anim Phys Sim Space Settings Structure=(SimSpaceAngularAlpha=0,MaxAngularVelocity=10000,MaxAngularAcceleration=10000,ExternalAngularVelocity=(X=0,Y=0,Z=0))
    ExternalForce="0,0,0"
   ) -> (Pose) {
    angularBiasOverride=0.9999
    angularDampingOverride=0.9999
    bLinearSpring=true
    bOverrideAngularBias=true
    bOverrideAngularDamping=true
    bOverrideLinearDamping=true
    bUseGravityOverride=true
    linearDampingOverride=0.9999
    linearSpringConstant=100
    simulationSpace=RootRelative
    alphaBoolBlend {blendOption=Linear}
    alphaScaleBiasClamp {scale=0.6}
    boundBone {boneName=dyn_fair_l_02}
    componentAppliedLinearAccClamp=<1250 1250 1250>
    componentLinearAccScale=<0.4 0.5 0.5>
    componentLinearVelScale=<0.4 0.5 0.5>
    physicsBodyDefinitions [{
      boundBone {boneName=dyn_fair_l_02}
      boxExtents=<8 8 12>
      constraintSetup {angularTargetAxis=AxisZ, angularLimitsMax=<5 15 1>, angularLimitsMin=<-5 -15 -15>, angularTarget=<1 1 1>}
      localJointOffset=<3 0 0>
     }]
   }
   n4 = Anim Dynamics - Bone: dyn_fair_l_03!(
    ComponentPose=@n3
    Alpha=1
    GravityOverride:Vector="0,0,0"
    SimSpaceSettings:Anim Phys Sim Space Settings Structure=(SimSpaceAngularAlpha=0,MaxAngularVelocity=10000,MaxAngularAcceleration=10000,ExternalAngularVelocity=(X=0,Y=0,Z=0))
    ExternalForce="0,0,0"
   ) -> (Pose) {
    angularBiasOverride=0.9999
    angularDampingOverride=0.9999
    bLinearSpring=true
    bOverrideAngularBias=true
    bOverrideAngularDamping=true
    bOverrideLinearDamping=true
    bUseGravityOverride=true
    linearDampingOverride=0.9999
    linearSpringConstant=100
    simulationSpace=RootRelative
    alphaBoolBlend {blendOption=Linear}
    alphaScaleBiasClamp {scale=0.58}
    boundBone {boneName=dyn_fair_l_03}
    componentAppliedLinearAccClamp=<1250 1250 1250>
    componentLinearAccScale=<0.4 0.5 0.5>
    componentLinearVelScale=<0.4 0.5 0.5>
    physicsBodyDefinitions [{
      boundBone {boneName=dyn_fair_l_03}
      boxExtents=<8 8 12>
      constraintSetup {angularTargetAxis=AxisZ, angularLimitsMax=<5 15 1>, angularLimitsMin=<-5 -15 -15>, angularTarget=<1 1 1>}
      localJointOffset=<4.5 0 0>
     }]
   }
   n5 = Anim Dynamics - Bone: dyn_fair_r_01!(
    ComponentPose=@n4
    Alpha=1
    GravityOverride:Vector="0,0,0"
    SimSpaceSettings:Anim Phys Sim Space Settings Structure=(SimSpaceAngularAlpha=0,MaxAngularVelocity=10000,MaxAngularAcceleration=10000,ExternalAngularVelocity=(X=0,Y=0,Z=0))
    ExternalForce="0,0,0"
   ) -> (Pose) {
    angularBiasOverride=0.9999
    angularDampingOverride=0.9999
    bLinearSpring=true
    bOverrideAngularBias=true
    bOverrideAngularDamping=true
    bOverrideLinearDamping=true
    bUseGravityOverride=true
    linearDampingOverride=0.9999
    linearSpringConstant=100
    simulationSpace=RootRelative
    alphaBoolBlend {blendOption=Linear}
    alphaScaleBiasClamp {scale=0.78}
    boundBone {boneName=dyn_fair_r_01}
    componentAppliedLinearAccClamp=<1250 1250 1250>
    componentLinearAccScale=<0.4 0.5 0.5>
    componentLinearVelScale=<0.4 0.5 0.5>
    physicsBodyDefinitions [{
      boundBone {boneName=dyn_fair_r_01}
      boxExtents=<8 8 12>
      constraintSetup {angularTargetAxis=AxisZ, angularLimitsMax=<5 15 1>, angularLimitsMin=<-5 -15 -15>, angularTarget=<1 1 1>}
      localJointOffset=<3 0 0>
     }]
   }
   n6 = Anim Dynamics - Bone: dyn_fair_r_02!(
    ComponentPose=@n5
    Alpha=1
    GravityOverride:Vector="0,0,0"
    SimSpaceSettings:Anim Phys Sim Space Settings Structure=(SimSpaceAngularAlpha=0,MaxAngularVelocity=10000,MaxAngularAcceleration=10000,ExternalAngularVelocity=(X=0,Y=0,Z=0))
    ExternalForce="0,0,0"
   ) -> (Pose) {
    angularBiasOverride=0.9999
    angularDampingOverride=0.9999
    bLinearSpring=true
    bOverrideAngularBias=true
    bOverrideAngularDamping=true
    bOverrideLinearDamping=true
    bUseGravityOverride=true
    linearDampingOverride=0.9999
    linearSpringConstant=100
    simulationSpace=RootRelative
    alphaBoolBlend {blendOption=Linear}
    alphaScaleBiasClamp {scale=0.72}
    boundBone {boneName=dyn_fair_r_02}
    componentAppliedLinearAccClamp=<1250 1250 1250>
    componentLinearAccScale=<0.4 0.5 0.5>
    componentLinearVelScale=<0.4 0.5 0.5>
    physicsBodyDefinitions [{
      boundBone {boneName=dyn_fair_r_02}
      boxExtents=<8 8 12>
      constraintSetup {angularTargetAxis=AxisZ, angularLimitsMax=<5 15 1>, angularLimitsMin=<-5 -15 -15>, angularTarget=<1 1 1>}
      localJointOffset=<4 0 0>
     }]
   }
   n7 = Anim Dynamics - Bone: dyn_fair_r_03!(
    ComponentPose=@n6
    Alpha=1
    GravityOverride:Vector="0,0,0"
    SimSpaceSettings:Anim Phys Sim Space Settings Structure=(SimSpaceAngularAlpha=0,MaxAngularVelocity=10000,MaxAngularAcceleration=10000,ExternalAngularVelocity=(X=0,Y=0,Z=0))
    ExternalForce="0,0,0"
   ) -> (Pose) {
    angularBiasOverride=0.9999
    angularDampingOverride=0.9999
    bLinearSpring=true
    bOverrideAngularBias=true
    bOverrideAngularDamping=true
    bOverrideLinearDamping=true
    bUseGravityOverride=true
    linearDampingOverride=0.9999
    linearSpringConstant=100
    simulationSpace=RootRelative
    alphaBoolBlend {blendOption=Linear}
    alphaScaleBiasClamp {scale=0.7}
    boundBone {boneName=dyn_fair_r_03}
    componentAppliedLinearAccClamp=<1250 1250 1250>
    componentLinearAccScale=<0.4 0.5 0.5>
    componentLinearVelScale=<0.4 0.5 0.5>
    physicsBodyDefinitions [{
      boundBone {boneName=dyn_fair_r_03}
      boxExtents=<8 8 12>
      constraintSetup {angularTargetAxis=AxisZ, angularLimitsMax=<5 15 1>, angularLimitsMin=<-5 -15 -15>, angularTarget=<1 1 1>}
      localJointOffset=<4 0 0>
     }]
   }
   n8 = Anim Dynamics - Bone: dyn_fair_r_04!(
    ComponentPose=@n7
    Alpha=1
    GravityOverride:Vector="0,0,0"
    SimSpaceSettings:Anim Phys Sim Space Settings Structure=(SimSpaceAngularAlpha=0,MaxAngularVelocity=10000,MaxAngularAcceleration=10000,ExternalAngularVelocity=(X=0,Y=0,Z=0))
    ExternalForce="0,0,0"
   ) -> (Pose) {
    angularBiasOverride=0.9999
    angularDampingOverride=0.9999
    bLinearSpring=true
    bOverrideAngularBias=true
    bOverrideAngularDamping=true
    bOverrideLinearDamping=true
    bUseGravityOverride=true
    linearDampingOverride=0.9999
    linearSpringConstant=100
    simulationSpace=RootRelative
    alphaBoolBlend {blendOption=Linear}
    alphaScaleBiasClamp {scale=0.6}
    boundBone {boneName=dyn_fair_r_04}
    componentAppliedLinearAccClamp=<1250 1250 1250>
    componentLinearAccScale=<0.4 0.5 0.5>
    componentLinearVelScale=<0.4 0.5 0.5>
    physicsBodyDefinitions [{
      boundBone {boneName=dyn_fair_r_04}
      boxExtents=<8 8 12>
      constraintSetup {angularTargetAxis=AxisZ, angularLimitsMax=<5 15 1>, angularLimitsMin=<-5 -15 -15>, angularTarget=<1 1 1>}
      localJointOffset=<3 0 0>
     }]
   }
   Output Pose(Result=Component To Local(ComponentPose=@n8)) {layerGroup=None}
  ]}
 EventGraph {nodes [
   Event BlueprintUpdateAnimation()
   > n1 = "Gravity Override State Function (Parameter Asset) (Nativized)"(
    CharacterPartAnimInstance=Self-Reference()
    GameTimeInSeconds=GetGameTimeInSeconds()
    Parameters=/BRCosmetics/Characters/Player/Male/Medium/Bodies/M_MED_Gimmick/Meshes/M_MED_Gimmick_Parameters
   )
   > Set GravityOverride(GravityOverride=@n1)
   n2 = NOT Boolean(A=Get bIsPlayingEmote())
   n3 = NOT Boolean(A=Get bIsSkyDiving())
   > n4 = Branch(Condition=AND Boolean(A=Get bIsInFrontEnd(), B=AND Boolean(A=@n3, B=@n2)))
   > Set SimSpaceSettings(SimSpaceSettings:Anim Phys Sim Space Settings Structure=Get SimSpaceSettingsFE() -> (SimSpaceSettingsFE:Anim Phys Sim Space Settings Structure)) -> (Output_Get:Anim Phys Sim Space Settings Structure)
   Set SimSpaceSettings(execute=@n4.else, SimSpaceSettings=Get SimSpaceSettingsPIE() -> (SimSpaceSettingsPIE:Anim Phys Sim Space Settings Structure))
  ]}
]
variables [
 Anim_Dynamics_Alpha_Frontend {type=number}
 GravityOverride {type=Vector}
 SimSpaceSettings {
  type=AnimPhysSimSpaceSettings
  default {maxAngularAcceleration=100000, maxAngularVelocity=100000, simSpaceAngularAlpha=1, externalAngularVelocity=<100000 100000 100000>}
 }
 SimSpaceSettingsFE {
  type=AnimPhysSimSpaceSettings
  default {maxAngularAcceleration=75, maxAngularVelocity=5, simSpaceAngularAlpha=0.7, externalAngularVelocity=<1 1 1>}
 }
 SimSpaceSettingsPIE {
  type=AnimPhysSimSpaceSettings
  default {maxAngularAcceleration=10, maxAngularVelocity=10, simSpaceAngularAlpha=0, externalAngularVelocity=<2 2 2>}
 }
]
